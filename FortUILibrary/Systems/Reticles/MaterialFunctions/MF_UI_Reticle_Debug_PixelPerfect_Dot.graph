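# UI material function that draws a pixel-consistent center dot with a drop shadow, filed as a reticle debug helper.
# It rescales UI UVs around the center by Pixel Size over Pixel Dimensions, then feeds a circle SDF through anti-aliasing for Final Color.
# A second circle, enlarged by Drop Shadow Size, is blended with the dot for Opacity Mask, leaving the shadow ring dark and partly transparent.
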MaterialFunction /FortUILibrary/Systems/Reticles/MaterialFunctions/MF_UI_Reticle_Debug_PixelPerfect_Dot
graph {
 properties {bExposeToLibrary=true, description="Provides a pixel perfect consistent center dot ", previewMaterialDomain=MD_UI}
 nodes [
  n1 = FunctionInput() {
   bUsePreviewValueAsDefault=true
   description=Expected Dot Size in Pixels
   inputName=Pixel Dimensions
   inputType=FunctionInput_Scalar
   previewValue=<1 1 0 1>
  }
  n2 = MaterialFunctionCall() {materialFunction=&GetUserInterfaceUV}
  n3 = Constant2Vector() {desc=Center, g=0.5, r=0.5}
  n4 = Add(A=Multiply(A=Divide(A=@n2.Pixel Size, B=@n1), B=Subtract(A=@n2.Normalized UV, B=@n3)), B=@n3)
  n5 = FunctionInput() {
   bUsePreviewValueAsDefault=true
   description="Scale of the Radius of the Dot. "
   inputName=DotScale
   inputType=FunctionInput_Scalar
   sortPriority=1
   previewValue=<6 0 0 1>
  }
  n6 = MaterialFunctionCall(DistanceField=MaterialFunctionCall!(UVs=@n4, Radius=@n5) {materialFunction=/FortUILibrary/MaterialFunctions/Shapes/SDFs/MF_SDF_Circle}.SDF) {materialFunction=&MF_UI_AntiAliasing}
  FunctionOutput(None=@n6.Inner) {outputName=Final Color}
  n7 = FunctionInput() {
   bUsePreviewValueAsDefault=true
   description=Controls the size of the drop shadow
   inputName=Drop Shadow Size
   inputType=FunctionInput_Scalar
   sortPriority=2
   previewValue=<5 0 0 1>
  }
  n8 = MaterialFunctionCall(UVs=@n4, Radius=Add(A=@n5, B=@n7)) {materialFunction=&MF_SDF_Circle}
  FunctionOutput(None=LinearInterpolate(
    A=@n6.Inner
    B=MaterialFunctionCall(DistanceField=@n8.SDF) {materialFunction=&MF_UI_AntiAliasing}.Inner
   ) {constAlpha=0.4}) {outputName=Opacity Mask}
 ]
}
