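# Material function that remaps weapon spread data into values usable by UI reticle materials.
# The Accuracy Spread Offset from the engine, which can be unbounded, is floored at a Min Value and added to a Positional Spread Offset that sets the bloom's starting position.
# RemapValueRange then maps the sum from the Input Range [Min,Max] to the Remapped Range [Min,Max], each supplied as a two-component vector.
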MaterialFunction /FortUILibrary/Systems/Reticles/MaterialFunctions/MF_UI_Reticle_Remap_Spread
graph {
 properties {bExposeToLibrary=true, description=Remaps the weapon spread data to something usable for reticle materials, previewMaterialDomain=MD_UI}
 nodes [
  n1 = FunctionInput() {
   bUsePreviewValueAsDefault=true
   desc=Way to positon before spread data
   description=Defines the starting position of the reticle bloom before the offset is added.
   inputName=Positional Spread Offset
   inputType=FunctionInput_Scalar
  }
  n2 = FunctionInput() {
   bUsePreviewValueAsDefault=true
   desc=Data from Engine for Spread - Can be Infinite
   description=Input for spread data variable.
   inputName=Accuracy Spread Offset
   inputType=FunctionInput_Scalar
   previewValue=<1 0 0 1>
  }
  n3 = FunctionInput() {bUsePreviewValueAsDefault=true, description=Clamps Min value for Data, inputName=Min Value, inputType=FunctionInput_Scalar, sortPriority=1}
  n4 = Max(A=@n2, B=@n3) {constB=0, desc="B = Min Value "}
  n5 = FunctionInput() {
   bUsePreviewValueAsDefault=true
   desc=Clamp Value for Input
   description="Sets Min-(x) and Max-(y) clamps for the spread variable."
   inputName="Input Range [Min,Max]"
   inputType=FunctionInput_Vector2
   sortPriority=2
   previewValue=<0 100 0 1>
  }
  n6 = FunctionInput() {
   bUsePreviewValueAsDefault=true
   desc=Clamp Value for Output
   description=Remaps the input data Min and Max to material friendly values.
   inputName="Remapped Range [Min,Max]"
   inputType=FunctionInput_Vector2
   sortPriority=3
   previewValue=<-1 1 0 1>
  }
  FunctionOutput(None=MaterialFunctionCall(
    Input=Add(A=@n1, B=@n4)
    Input Low=ComponentMask(None=@n5) {r=true}
    Input High=ComponentMask(None=@n5) {g=true}
    Target Low=ComponentMask(None=@n6) {r=true}
    Target High=ComponentMask(None=@n6) {g=true}
   ) {desc=Remap Data to MF Values, materialFunction=&RemapValueRange}.Result)
 ]
}
