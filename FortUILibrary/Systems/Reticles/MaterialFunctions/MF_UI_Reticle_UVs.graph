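# Material function that builds UVs for radially mirrored UI reticles or reticle segments, with spread-driven bloom.
# Input UVs are rotated and scaled overall, radially mirrored by a floored side count, then rotated and scaled again per prong.
# Positional and spread offsets are summed, remapped between configurable ranges and added to the prong V coordinate, shifting its distance from center.
# Rotators use a 2π speed, so rotation inputs are in turns.
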MaterialFunction /FortUILibrary/Systems/Reticles/MaterialFunctions/MF_UI_Reticle_UVs
graph {
 properties {
  bExposeToLibrary=true
  description=UV Manipulations for creating radial mirrored reticles or reticle segments.
  previewBlendMode=BLEND_Translucent
  previewMaterialDomain=MD_UI
 }
 nodes [
  n1 = FunctionInput(Preview=MaterialFunctionCall() {materialFunction=&GetUserInterfaceUV}.Normalized UV) {
   bUsePreviewValueAsDefault=true
   description="Optional custom UV mapping (defaults to Normalized UVs)"
   inputName=Reticle UVs
   inputType=FunctionInput_Vector2
   sortPriority=-1
  }
  n2 = FunctionInput() {
   bUsePreviewValueAsDefault=true
   description="Overall Rotation of the reticle  UV's "
   inputName=Reticle Rotation
   inputType=FunctionInput_Scalar
   sortPriority=2
   previewValue=<0.5 0 0 1>
  }
  n3 = FunctionInput() {
   bUsePreviewValueAsDefault=true
   description=Scales the overall size of the reticle.
   inputName=Reticle Scale
   inputType=FunctionInput_Scalar
   sortPriority=1
   previewValue=<1 0 0 1>
  }
  n4 = MaterialFunctionCall(UVs=Rotator(Coordinate=@n1, Time=@n2) {speed=6.2832}, Texture Scale=@n3) {materialFunction=&ScaleUVsByCenter}
  n5 = FunctionInput() {
   bUsePreviewValueAsDefault=true
   description=The number of sides to mirror.
   inputName=Reticle Sides
   inputType=FunctionInput_Scalar
   previewValue=<4 0 0 1>
  }
  n6 = Floor(None=@n5)
  n7 = FunctionInput() {
   bUsePreviewValueAsDefault=true
   description="Rotation of individual prongs. "
   inputName=Prong Rotation
   inputType=FunctionInput_Scalar
   sortPriority=4
  }
  n8 = FunctionInput() {
   bUsePreviewValueAsDefault=true
   description=Scale of Individual prongs.
   inputName=Prong Scale
   inputType=FunctionInput_Scalar
   sortPriority=3
   previewValue=<1 0 0 1>
  }
  n9 = MaterialFunctionCall(
   UVs=Rotator(
    Coordinate=MaterialFunctionCall!(UVs=@n4.UVs, Sides=@n6) {materialFunction=/FortUILibrary/MaterialFunctions/Basics/MF_UI_UVRadialMirror}.UVs
    Time=@n7
   ) {speed=6.2832}
   Texture Scale=@n8
  ) {materialFunction=&ScaleUVsByCenter}
  n10 = ComponentMask(None=@n9.UVs) {desc=Distance from Center, g=true}
  n11 = FunctionInput() {
   bUsePreviewValueAsDefault=true
   description=Defines the starting position of the reticle bloom before the offset is added.
   inputName=Positional Offset
   inputType=FunctionInput_Scalar
   sortPriority=5
  }
  n12 = FunctionInput() {
   bUsePreviewValueAsDefault=true
   description=Input for spread data variable.
   inputName=Spread Offset
   inputType=FunctionInput_Scalar
   sortPriority=5
  }
  n13 = FunctionInput() {
   bUsePreviewValueAsDefault=true
   description="Sets Min-(x) and Max-(y) clamps for the spread variable."
   inputName="Spread Input Range [Min,Max]"
   inputType=FunctionInput_Vector2
   sortPriority=6
   previewValue=<0 1 0 1>
  }
  n14 = FunctionInput() {
   bUsePreviewValueAsDefault=true
   description=Remaps the input data Min and Max to material friendly values.
   inputName="Spread Remapped Range [Min,Max]"
   inputType=FunctionInput_Vector2
   sortPriority=7
   previewValue=<0 1 0 1>
  }
  n15 = MaterialFunctionCall(
   Input=Add(A=@n11, B=@n12)
   Input Low=ComponentMask(None=@n13) {r=true}
   Input High=ComponentMask(None=@n13) {g=true}
   Target Low=ComponentMask(None=@n14) {r=true}
   Target High=ComponentMask(None=@n14) {g=true}
  ) {desc=Remap Spread Data to MF Values, materialFunction=&RemapValueRange}
  FunctionOutput(None=AppendVector(A=ComponentMask(None=@n9.UVs) {desc="Skew around Center ", r=true}, B=Add(A=@n10, B=@n15.Result))) {outputName=UVs}
 ]
}
